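# Translucent unlit material for the crystal ribbon effect on the TigerRootHype pickaxe, enabled for Niagara ribbons.
# Emissive adds panning line streaks to two colors blended by a time-animated sine pulse of a crystal texture and tinted by a curve-atlas gradient along U, all scaled by particle color.
# Opacity combines UV edge fades, crystal noise eroded along U and particle alpha, with depth fade skipped on forward and mobile paths.
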Material /BRCosmetics/Weapons/FORT_Melee/Pickaxe_TigerRootHype/FX/M_Pickaxe_TigerRootHype_Crystal
graph {
 properties {bUsedWithNiagaraRibbons=true, blendMode=BLEND_Translucent, pixelDepthOffsetMode=PDOM_Legacy, shadingModel=MSM_Unlit, usedShadingModels=Unlit}
 nodes [
  n1 = Panner(Coordinate=TextureCoordinate() {uTiling=4}) {speedX=-1.5}
  n2 = Multiply(A=TextureSample(UVs=@n1) {texture=&/T_Pickaxe_TigerRootHype_line}.R) {constB=2}
  n3 = VectorParameter() {parameterName=Color02, defaultValue=<rgba 1 0 0.9533 1>}
  n4 = VectorParameter() {parameterName=Color01, defaultValue=<rgba 2 2 2 1>}
  n5 = Add(A=TextureSample(UVs=@n1) {texture=/BRCosmetics/Weapons/FORT_Melee/Pickaxe_Majesty_Male/FX/T_Crystal}.R, B=Multiply(A=Time()) {constB=-1.5})
  n6 = Power(Base=Abs(None=Sine(None=@n5))) {constExponent=5}
  n7 = Power(Base=Multiply(A=Add(B=@n6)))
  n8 = CurveAtlasRowParameter(CurveTime=ComponentMask(None=TextureCoordinate()) {r=true}) {
   atlas=/Game/Athena/LinearColorCurves/Curve_Atlas_Contrail_Inferno
   curve=/Game/Weapons/FORT_Melee/Pickaxe_TigerRootHype/FX/Pickaxe_TigerRootHype_ColorCurve
   defaultValue=2
   parameterName=Color Curve
  }
  n9 = Multiply(A=Multiply(A=LinearInterpolate(A=@n3.RGB, B=@n4.RGB, Alpha=@n7), B=@n8))
  n10 = ParticleColor()
  MaterialFunctionCall(Emissive=Multiply(A=Add(A=@n2, B=@n9), B=@n10.RGB)) {materialFunction=&MF_FX_EmissiveAdjustments}
  n11 = OneMinus(None=ComponentMask(None=TextureCoordinate()) {r=true})
  n12 = Multiply(A=ComponentMask(None=TextureCoordinate()) {r=true}, B=Saturate(None=Power(Base=@n11) {constExponent=1}))
  n13 = Sine(None=ComponentMask(None=TextureCoordinate()) {g=true}) {period=2}
  n14 = Saturate(None=Multiply(A=Power(Base=@n13)) {constB=30})
  n15 = Multiply(B=Multiply(A=Multiply(A=@n12, B=@n14)) {constB=5}) {constA=1}
  n16 = Panner(Coordinate=TextureCoordinate() {uTiling=4}) {speedX=0.5}
  n17 = ComponentMask(None=TextureCoordinate()) {r=true}
  n18 = LinearInterpolate(Alpha=Power(Base=OneMinus(None=@n17)) {constExponent=1.5})
  n19 = Subtract(
   A=Power(Base=TextureSample(UVs=@n16) {texture=&T_Crystal}.R, Exp=@n18)
   B=Saturate(None=@n18)
  )
  n20 = Multiply(A=@n10.A, B=Multiply(A=Multiply(A=@n15, B=@n19)))
  n21 = DepthFade(Opacity=@n20) {fadeDistanceDefault=10}
  ShadingPathSwitch(Default=@n21, Deferred=@n21, Forward=@n20, Mobile=@n20)
 ]
}
