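# Translucent, unlit, two-sided Niagara mesh-particle material for the crystal impact effect of the TigerRootHype pickaxe.
# Two tiled crystal texture layers shift over particle lifetime, with contrast varying along V and a gradient mask fading one end.
# Their sum is tinted by a V-indexed curve atlas row and particle color for emissive.
# A differently weighted sum times particle alpha drives opacity, with depth fade omitted on forward and mobile paths.
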Material /BRCosmetics/Weapons/FORT_Melee/Pickaxe_TigerRootHype/FX/M_Pickaxe_TigerRootHype_Crystal_Impact
graph {
 properties {
  bUsedWithNiagaraMeshParticles=true
  blendMode=BLEND_Translucent
  pixelDepthOffsetMode=PDOM_Legacy
  shadingModel=MSM_Unlit
  twoSided=true
  usedShadingModels=Unlit
 }
 nodes [
  n1 = ParticleColor()
  n2 = CurveAtlasRowParameter(CurveTime=ComponentMask(None=TextureCoordinate()) {g=true}) {
   atlas=/Game/Athena/LinearColorCurves/Curve_Atlas_Contrail_Inferno
   curve=/Game/Weapons/FORT_Melee/Pickaxe_TigerRootHype/FX/Pickaxe_TigerRootHype_ColorCurve_2
   defaultValue=3
  }
  n3 = VectorParameter() {parameterName=UV Offset01, defaultValue=<rgba 0.3 0.3 0 1>}
  n4 = LinearInterpolate(Alpha=ParticleRelativeTime()) {constA=0.2, constB=0.35}
  n5 = TextureSample(UVs=Add(A=TextureCoordinate() {uTiling=5}, B=AppendVector(A=@n3.R, B=@n4))) {samplerType=SAMPLERTYPE_LinearGrayscale, texture=&/T_Pickaxe_TigerRootHype_Crystal}
  n6 = ComponentMask(None=TextureCoordinate()) {g=true}
  n7 = LinearInterpolate(Alpha=Power(Base=OneMinus(None=@n6))) {constA=2, constB=25}
  n8 = ComponentMask(None=TextureCoordinate()) {g=true}
  n9 = Power(Base=Add(A=OneMinus(None=@n8)) {constB=-0.1})
  n10 = Multiply(A=Power(Base=Power(Base=@n5.R, Exp=@n7)), B=Saturate(None=Multiply(A=@n9) {constB=30}))
  n11 = Multiply(A=@n10) {constB=0.2}
  n12 = VectorParameter() {parameterName=UV offset02, defaultValue=<rgba 0 0.9 0 1>}
  n13 = LinearInterpolate(Alpha=ParticleRelativeTime()) {constA=0.9}
  n14 = TextureSample(UVs=Add(A=TextureCoordinate() {uTiling=4}, B=AppendVector(A=@n12.R, B=@n13))) {samplerType=SAMPLERTYPE_LinearGrayscale, texture=&T_Pickaxe_TigerRootHype_Crystal}
  n15 = OneMinus(None=ComponentMask(None=TextureCoordinate()) {g=true})
  n16 = Power(Base=@n14.R, Exp=LinearInterpolate(Alpha=Power(Base=@n15)) {constA=2, constB=25})
  n17 = OneMinus(None=ComponentMask(None=TextureCoordinate()) {g=true})
  n18 = Multiply(A=Power(Base=Add(A=@n17) {constB=-0.1})) {constB=30}
  n19 = Multiply(A=Multiply(A=Power(Base=@n16), B=Saturate(None=@n18))) {constB=0.5}
  n20 = Multiply(A=@n2, B=Add(B=Add(A=@n19, B=@n11)))
  MaterialFunctionCall(Emissive=Multiply(A=@n1.RGB, B=Multiply(A=@n20) {constB=3})) {materialFunction=&MF_FX_EmissiveAdjustments}
  n21 = Multiply(A=@n10) {constB=20}
  n22 = Multiply(A=@n1.A, B=Multiply(A=Add(A=@n21, B=@n19) {constB=0}))
  n23 = DepthFade(Opacity=@n22) {fadeDistanceDefault=20}
  ShadingPathSwitch(Default=@n23, Deferred=@n23, Forward=@n22, Mobile=@n22)
 ]
}
