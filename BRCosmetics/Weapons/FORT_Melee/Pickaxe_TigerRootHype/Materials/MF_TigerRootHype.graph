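# Material function for the TigerRootHype pickaxe that layers an animated, audio-reactive glow over incoming material attributes, behind a static switch.
# Position thresholds define top, mid and spike masks, combined with FX mask textures, time-driven sine noise and warping.
# Event and emote audio levels from Fortnite's parameter collection modulate the sections.
# Fresnel-blended colors, shine and spike glow pass through a time-of-day curve into emissive, alongside overridden surface values and detail normals.
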MaterialFunction /BRCosmetics/Weapons/FORT_Melee/Pickaxe_TigerRootHype/Materials/MF_TigerRootHype
graph {nodes [
  n1 = ComponentMask(None=NamedRerouteUsage()) {b=true}
  n2 = Multiply(A=Add(A=@n1, B=Constant() {r=-0.6225}), B=Constant() {r=2.43})
  n3 = Ceil(None=Saturate(None=NamedRerouteDeclaration(None=@n2) {name=Msk_Top_Grad, nodeColor=<rgba 0.2891 1 0 1>}))
  Add(
   A=NamedRerouteDeclaration(None=@n3) {name=Msk_Top, nodeColor=<rgba 0.2891 1 0 1>}
   B=ScalarParameter() {defaultValue=0.5, group=Mat_TigerRootHype, parameterName=TigR_Color_B Opacity}
  ) {constB=0.5}
  n4 = NamedRerouteUsage()
  n5 = Multiply(A=Add(A=ComponentMask(None=@n4) {b=true}, B=Constant() {r=-0.095}), B=Constant() {r=2.03})
  Ceil(None=Saturate(None=NamedRerouteDeclaration(None=@n5) {name=Msk_Mid_Grad, nodeColor=<rgba 0.2891 1 0 1>}))
  n6 = OneMinus(None=Fresnel(ExponentIn=ScalarParameter() {defaultValue=0.5, group=TigerRoot, parameterName=TigR_Shine_Fresnel_Expo}) {baseReflectFraction=0, exponent=0.5})
  n7 = ComponentMask(None=NamedRerouteUsage()) {b=true}
  n8 = Multiply(A=VectorParameter() {group=TigerRoot_Colors, parameterName=TigR_Color_Edge_Glow, defaultValue=<rgba 0.2708 0.9421 1 1>}.RGB, B=@n7)
  n9 = LinearInterpolate(
   A=VectorParameter() {group=TigerRoot_Colors, parameterName=TigR_Color_Aa, defaultValue=<rgba 0 0.008 0.2552 1>}.RGB
   B=VectorParameter() {group=TigerRoot_Colors, parameterName=TigR_Color_Ab, defaultValue=<rgba 0 0.0023 0.0729 1>}.RGB
   Alpha=Fresnel(ExponentIn=ScalarParameter() {defaultValue=0.5, group=Mat_TigerRootHype, parameterName=TigR_Fresnel_A_Expo}) {baseReflectFraction=0, exponent=0.5}
  )
  n10 = LinearInterpolate(
   A=VectorParameter() {group=TigerRoot_Colors, parameterName=TigR_Color_Ba, defaultValue=<rgba 0.217 0 0.6 1>}.RGB
   B=VectorParameter() {group=TigerRoot_Colors, parameterName=TigR_Color_Bb, defaultValue=<rgba 0.0182 0 0.099 1>}.RGB
   Alpha=Fresnel(ExponentIn=ScalarParameter() {defaultValue=0.5, group=Mat_TigerRootHype, parameterName=TigR_Fresnel_B_Expo}) {baseReflectFraction=0, exponent=0.5}
  )
  n11 = Power(Base=NamedRerouteUsage())
  n12 = ScalarParameter() {defaultValue=3, group=Mat_TigerRootHype, parameterName=TigR_Top_Grad_Brightness}
  n13 = NamedRerouteUsage()
  n14 = Time()
  n15 = ScalarParameter() {defaultValue=0.5, group=TigerRoot, parameterName=TigR_FX_Noise_Speed}
  n16 = Sine(None=Add(A=ComponentMask(None=@n13) {a=true}, B=Multiply(A=@n14, B=@n15)))
  n17 = ConstantBiasScale(None=@n16)
  n18 = LinearInterpolate(A=ScalarParameter() {defaultValue=0.75, group=TigerRoot, parameterName=TigR_FX_Noise_Strength}, Alpha=@n17)
  n19 = Multiply(A=Multiply(A=Multiply(A=@n11, B=@n12), B=Saturate(None=@n18)))
  n20 = NamedRerouteUsage()
  n21 = Multiply(A=Saturate(None=Add(A=@n19) {constB=0}), B=Power(Base=OneMinus(None=@n20)) {constExponent=3})
  n22 = NamedRerouteUsage()
  n23 = NamedRerouteUsage()
  n24 = Multiply(
   A=Saturate(None=Add(A=@n22, B=@n23))
   B=VectorParameter() {group=TigerRoot_Colors, parameterName=TigR_Color_Specturm, defaultValue=<rgba 0 0.6905 0.75 1>}.RGB
  )
  n25 = ScalarParameter() {defaultValue=1, group=Mat_TigerRootHype, parameterName=TigR_Glow_Fresnel_Expo}
  n26 = Power(
   Base=OneMinus(None=Fresnel(ExponentIn=@n25) {baseReflectFraction=0, exponent=0.5})
   Exp=ScalarParameter() {defaultValue=2, group=Mat_TigerRootHype, parameterName=TigR_Glow_Fresnel_Pow}
  ) {constExponent=3}
  n27 = Multiply(A=OneMinus(None=@n7), B=Add(A=LinearInterpolate(A=@n9, B=@n10, Alpha=@n21), B=Multiply(A=@n24, B=@n26) {constA=1}))
  n28 = LinearInterpolate(
   A=VectorParameter() {group=TigerRoot_Colors, parameterName=TigR_Color_Spike_A, defaultValue=<rgba 0.5 0.9603 1 1>}.RGB
   B=VectorParameter() {group=TigerRoot_Colors, parameterName=TigR_Color_Spike_B, defaultValue=<rgba 0.8125 2.8263 3 1>}.RGB
   Alpha=NamedRerouteUsage()
  )
  n29 = Saturate(None=Add(A=@n26) {constB=0.25})
  n30 = Add(
   A=Multiply(
    A=VectorParameter() {group=TigerRoot_Colors, parameterName=TigR_Shine_Color, defaultValue=<rgba 0.1809 0 0.5 1>}.RGB
    B=Power(Base=@n6) {constExponent=10}
   )
   B=LinearInterpolate(A=Add(A=@n8, B=@n27), B=Multiply(A=@n28, B=@n29), Alpha=NamedRerouteUsage())
  )
  n31 = SetMaterialAttributes!(
   MaterialAttributes=NamedRerouteUsage()
   BaseColor=VectorParameter() {group=TigerRoot, parameterName=TigR_BaseColor, defaultValue=<rgba 0 0 0 1>}.RGB
   Metallic=ScalarParameter() {group=TigerRoot, parameterName=TigR_Metalic}
   Roughness=ScalarParameter() {group=TigerRoot, parameterName=TigR_Roughness}
   Specular=ScalarParameter() {defaultValue=0.5, group=TigerRoot, parameterName=TigR_Spec}
   EmissiveColor=MaterialFunctionCall!(Color=Clamp(None=@n30) {maxDefault=20}) {materialFunction=/Game/Packages/Fortress_SharedMaterials/Anime_Master_Material/MF_TODColorCurve}.Result
   Normal=NamedRerouteUsage()
  ) {attributeSetTypes [
    #1
    #2
    #3
    #4
    #5
    #6
   ]}
  FunctionOutput(None=StaticSwitchParameter(True=@n31, False=NamedRerouteUsage()) {defaultValue=true, group=Mat_TigerRootHype, parameterName=TigerRootHtpe})
  n32 = Add(B=OneMinus(None=NamedRerouteUsage())) {constA=-0.1}
  n33 = Add(A=@n32, B=Multiply(A=NamedRerouteUsage(), B=ScalarParameter() {defaultValue=0.75, group=TigerRoot_Mid, parameterName=TigR_Mid_Audio_Amount}))
  n34 = NamedRerouteUsage()
  n35 = Multiply(
   A=OneMinus(None=NamedRerouteUsage())
   B=LinearInterpolate(A=SmoothStep(Value=@n32), B=SmoothStep(Value=@n33) {constMin=0.8}, Alpha=Ceil(None=@n34))
  )
  Multiply(A=Ceil(None=NamedRerouteUsage()), B=@n35)
  n36 = MaterialFunctionCall() {materialFunction=/Game/Packages/Fortress_SharedMaterials/Base_Material_Functions/BaseCharacterShaderFunctions/Audio/MF_AudioReactivity}
  Multiply(A=@n36.Amplitude, B=@n36.Amplitude)
  NamedRerouteDeclaration(None=FunctionInput(Preview=WorldPosition().XYZ) {bUsePreviewValueAsDefault=true, inputName=PreSkin_Pos}) {name=PreSkin Pos, nodeColor=<rgba 0 0.4609 1 1>}
  n37 = TextureObjectParameter() {
   group=TigerRoot
   parameterName=TigR__Tex_FX_Norm
   samplerType=SAMPLERTYPE_Normal
   texture=/BRCosmetics/Weapons/FORT_Melee/Pickaxe_TigerRootHype/Textures/T_PickAxe_TigerRootHype_FX_N
  }
  NamedRerouteDeclaration(None=@n37) {name=Tex_Normals, nodeColor=<rgba 0 1 0.3203 1>}
  n38 = NamedRerouteUsage()
  n39 = NamedRerouteUsage()
  n40 = LinearInterpolate(
   A=ScalarParameter() {defaultValue=0.1, group=TigerRoot_Top, parameterName=TigR_Top_Warp_Off}
   B=ScalarParameter() {defaultValue=0.2, group=TigerRoot_Top, parameterName=TigR_Top_Warp_On}
   Alpha=Ceil(None=NamedRerouteUsage())
  )
  n41 = Add(A=NamedRerouteUsage(), B=Saturate(None=Multiply(A=@n39, B=@n40)))
  n42 = Multiply(A=NamedRerouteUsage(), B=ScalarParameter() {defaultValue=0.6, group=TigerRoot_Audio, parameterName=TigR_Top_Audio_Amount})
  n43 = NamedRerouteUsage()
  n44 = SmoothStep(Value=Add(A=OneMinus(None=@n41), B=LinearInterpolate(B=@n42, Alpha=@n43)) {constB=0}) {constMin=0.6}
  n45 = Add(A=ComponentMask(None=NamedRerouteUsage()) {r=true}) {constB=0.5}
  n46 = Multiply(A=Power(Base=Saturate(None=@n44)), B=LinearInterpolate(B=Saturate(None=@n45), Alpha=NamedRerouteUsage()) {constA=1})
  NamedRerouteDeclaration(None=Saturate(None=Multiply(A=@n38, B=@n46))) {name=Section_Top, nodeColor=<rgba 0 1 0.4141 1>}
  n47 = Constant() {r=-0.58}
  n48 = Power(Base=Divide(A=Add(A=@n1, B=@n47)) {constB=0.04})
  NamedRerouteDeclaration(None=Saturate(None=Add(A=@n3, B=@n48))) {name=Msk_Top_B, nodeColor=<rgba 0.2891 1 0 1>}
  NamedRerouteDeclaration(None=Saturate(None=@n35)) {name=Section_Mid, nodeColor=<rgba 1 0 0.7266 1>}
  n49 = Multiply(A=Constant3Vector() {constant=<rgba 1 1 0 0>}, B=TextureSample(Tex=NamedRerouteUsage()) {samplerType=SAMPLERTYPE_Normal}.RGB)
  n50 = ScalarParameter() {defaultValue=0.25, group=Mat_TigerRootHype, parameterName=TigR_FX_Normal_amount}
  NamedRerouteDeclaration(None=Add(A=NamedRerouteUsage(), B=Multiply(A=@n49, B=@n50))) {name=Normals, nodeColor=<rgba 0.4531 1 0 1>}
  NamedRerouteDeclaration(None=@n17) {name=Fx_Noise, nodeColor=<rgba 0 0.1094 1 1>}
  n51 = ScalarParameter() {defaultValue=1, group=TigerRoot_Audio, parameterName="TigR_Audio_Disable 0=OFF"}
  n52 = CollectionParameter() {collection=&FortniteMaterialParameters, parameterName=EventAudio_Amplitude}
  n53 = CollectionParameter() {collection=&FortniteMaterialParameters, parameterName=Emote_AudioLow}
  n54 = Max(A=Min(A=@n51, B=@n52), B=Min(A=@n51, B=@n53))
  n55 = MaterialFunctionCall(
   Input=@n54
   Input Low=ScalarParameter() {defaultValue=0.25, group=TigerRoot_Audio, parameterName=TigR_Sectrum_Low_Remap, sortPriority=150}
   Input High=ScalarParameter() {defaultValue=1, group=TigerRoot_Audio, parameterName=TigR_Sectrum_High_Remap, sortPriority=150}
   Target Low=Constant()
   Target High=Constant() {r=1}
  ) {materialFunction=&RemapValueRange}
  n56 = CollectionParameter() {collection=&FortniteMaterialParameters, parameterName=Emote_AudioHigh}
  n57 = Min(A=@n51, B=@n56)
  NamedRerouteDeclaration(None=Max(A=Saturate(None=@n55.Result), B=@n57)) {name=Audio_Data, nodeColor=<rgba 0.2891 0 1 1>}
  NamedRerouteDeclaration(None=@n16) {name=Fx_Noise_B, nodeColor=<rgba 0 0.1094 1 1>}
  n58 = Add(A=@n1, B=Constant() {r=-0.98})
  NamedRerouteDeclaration(None=Saturate(None=Ceil(None=@n58))) {name=Msk_Spike, nodeColor=<rgba 0.2891 1 0 1>}
  n59 = VertexInterpolator(VS=PreSkinnedPosition())
  n60 = PreSkinnedLocalBounds()
  n61 = MaterialFunctionCall(Color=Subtract(A=@n60.Max, B=@n60.Max)) {materialFunction=&SplitComponents}
  n62 = Max(A=@n61.R, B=@n61.R)
  NamedRerouteDeclaration(None=Divide(A=Subtract(A=@n59.PS, B=@n60.Min), B=Max(A=@n62, B=@n61.B))) {name=Cal_Bounds, nodeColor=<rgba 0.7109 0 1 1>}
  NamedRerouteDeclaration(None=Max(A=@n54, B=@n57)) {name=AudioSpectrum, nodeColor=<rgba 1 0.8203 0 1>}
  n63 = TextureSample() {samplerType=SAMPLERTYPE_Normal, texture=/BRCosmetics/Weapons/FORT_Melee/Pickaxe_TigerRootHype/Textures/Pickaxe_TigerRootHype_normal}
  n64 = GetMaterialAttributes(None=FunctionInput(Preview=SetMaterialAttributes!(MaterialAttributes, Normal=@n63.RGB) {attributeSetTypes [#6]}) {inputType=FunctionInput_MaterialAttributes}) {attributeGetTypes [#5, #6]}
  NamedRerouteDeclaration(None=@n64.MaterialAttributes) {name=Root_Material, nodeColor=<rgba 0.5234 0 1 1>}
  NamedRerouteDeclaration(None=@n64.EmissiveColor) {name=Root_Emissive, nodeColor=<rgba 0.4297 1 0 1>}
  n65 = TextureObjectParameter() {
   group=TigerRoot
   parameterName=TigR__Tex_FxMap
   samplerType=SAMPLERTYPE_Masks
   texture=/BRCosmetics/Weapons/FORT_Melee/Pickaxe_TigerRootHype/Textures/T_PickAxe_TigerRootHype_FX_Msk
  }
  NamedRerouteDeclaration(None=@n65) {name=Tex_Fx_Map, nodeColor=<rgba 0 1 0.3203 1>}
  NamedRerouteDeclaration(None=@n64.Normal) {name=Root_Normals, nodeColor=<rgba 1 0.7266 0 1>}
  NamedRerouteDeclaration(None=TextureSample(Tex=NamedRerouteUsage()) {samplerType=SAMPLERTYPE_Masks}.RGBA) {name=FX_Msk, nodeColor=<rgba 0.3828 1 0 1>}
  NamedRerouteUsage()
  NamedRerouteUsage()
 ]}
