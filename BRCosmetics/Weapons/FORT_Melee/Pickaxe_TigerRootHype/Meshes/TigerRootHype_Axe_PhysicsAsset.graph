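# Physics asset for the TigerRootHype pickaxe, a two-handed melee cosmetic weapon mesh.
# It holds a single capsule body on the Melee_TwoHanded_Handle bone, offset along the handle and rotated to cover the axe.
# There are no constraints, so the weapon collides and simulates as one rigid piece.
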PhysicsAsset /BRCosmetics/Weapons/FORT_Melee/Pickaxe_TigerRootHype/Meshes/TigerRootHype_Axe_PhysicsAsset
bodies {Melee_TwoHanded_Handle [{shapeType=Capsule, length=72.7746, radius=22.3526, center=<0.6186 -1.2622 39.5155>, rotation=<rot 0.7732 -90 0.002>}]}
constraints []
